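; Distant fog and iceberg event actor that brings an iceberg mesh closer along a spline, hidden at play start.
; Without cosmetic execution (server), it waits for match start, then derives MoveProgress from two calendar events, stopping at the 6-3 endpoint.
; OnRep_MoveProgress updates visibility, spline position, fog material distance, ice floe visibility and ambient audio parameters.
; An Approach event sets progress directly, and the construction script previews it.
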(event Custom|Approach
  (params (Position :type Float%20%28double-precision%29))
  (Variables|Default|SetMoveProgress Position))

(event EventBeginPlay
  ;[1 1 1] Begin Play
  (Rendering|SetActorHiddenInGame true)
  (Networking|CanExecuteCosmeticEvents
    (:False
      (Variables|Default|SetMoveProgress 1.01)
      (bind _asfort_game_mode_zone (Utilities|Casting|CastToFortGameModeZone (Game|GetGameMode)))
      (EventDispatchers|BindEventtoOnHandleMatchHasStarted _asfort_game_mode_zone (EventDispatchers|CreateEvent :!event Custom|ContinueSetup)))))

(event Custom|ContinueSetup
  (bind _6_3endpoint (Variables|Default|Get6-3EndPoint))
  ;[1 1 1] Calendar Event
  (Utilities|FlowControl|Delay 1.0)
  (Utilities|FlowControl|Sequence
    (:then_0
      (bind (_eventname _timeuntilend _timesincebegin) (MCPContext|WatchCalendarEvent (Variables|Default|GetFogApproachEvent))
        (:OnEventActive
          (Variables|Default|SetTotalEventTime (Math|Timespan|Timespan+Timespan _timeuntilend _timesincebegin))
          (Variables|Default|SetMoveProgress (Math|Float|MapRangeClamped (Math|Timespan|TimespanRatio _timeuntilend (Variables|Default|GetTotalEventTime)) 1.0 0.0 1.0 _6_3endpoint)) ; 0 = end point
        )))
    (:then_1
      (MCPContext|WatchCalendarEvent (Variables|Default|GetFogApproachOverlapEvent)
        (:OnEventActive
          (Variables|Default|SetMoveProgress _6_3endpoint))))))

(fn OnRep_MoveProgress
  (bind _ambaudio (Variables|Default|GetAmbAudio))
  (Rendering|SetActorHiddenInGame (> (Variables|Default|GetMoveProgress) 1.0))
  (CallFunction|UpdateIcebergLocation (Variables|Default|GetMoveProgress))
  (Audio|Parameter|SetFloatParameter _ambaudio "Progress" (Variables|Default|GetMoveProgress) :!class /Script/Engine.AudioComponent)
  (Audio|Parameter|SetFloatParameter _ambaudio "ProgressInv" (- 1.0 (Variables|Default|GetMoveProgress)) :!class /Script/Engine.AudioComponent)
  (if (not (Audio|Components|Audio|IsPlaying _ambaudio))
    (Audio|Components|Audio|Play _ambaudio 0.0)))

(fn UpdateIcebergLocation
  (params (Time :type Float%20%28double-precision%29))
  (Transformation|SetWorldLocation (Variables|Default|GetStaticMesh) (Spline|GetLocationatTime (Variables|Default|GetSpline) Time "World" true) false false)
  (Rendering|Material|SetScalarParameterValueonMaterials (Variables|Default|GetSM_Ice_Fog_Top) "Distance" Time)
  (Rendering|SetVisibility (Variables|Default|GetIceFloeParent) (< Time (Variables|Default|Get6-3EndPoint)) true))

(fn ConstructionScript
  (CallFunction|UpdateIcebergLocation (Variables|Default|GetPreviewMoveProgress)))
