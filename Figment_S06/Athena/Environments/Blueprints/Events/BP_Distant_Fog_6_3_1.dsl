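; Distant iceberg and fog set piece that approaches the island during a timed calendar event.
; The server derives replicated MoveProgress from the event's remaining time after match start, and an overlap event pins it to the 6.31 endpoint.
; OnRep moves the mesh along a spline, drives fog, rock, ice and aurora visuals plus ambient audio, and spawns snow when close.
; On StormHolding, clients fade playing audio, otherwise deactivate snow.
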(event Custom|Approach
  (params (Position :type Float%20%28double-precision%29))
  ;[1 1 1] Debug
  (Variables|Default|SetMoveProgress Position))

(event EventBeginPlay
  ;[1 1 1] Begin Play
  (Networking|CanExecuteCosmeticEvents
    (:True
      ;[1 1 1] Fade out audio on client when storm shows up
      (bind (_asynctaskproxy _payload) (GameplayEventRouter|StartListeningToStatefulEvent (GameplayEventRouter|GetGlobalEventRouter) "/Script/FortniteGame.FortGamePhaseStepUpdatedEvent")
        (:then
          (Variables|Default|SetAsync%20Task%20Ref _asynctaskproxy))
        (:OnEventReceived
          _)
        (:OnSavedEventStateExists
          _))
      (switch Utilities|FlowControl|Switch|SwitchonEAthenaGamePhaseStep (Utilities|Struct|BreakFortGamePhaseStepUpdatedEvent _payload)
        (:StormHolding
          (GameplayEventRouter|StopListeningtoStatefulEvent (Variables|Default|GetAsync%20Task%20Ref))
          (if (Audio|Components|Audio|IsPlaying (Variables|Default|GetAmbAudio))
            (Audio|Components|Audio|FadeOut (Variables|Default|GetAmbAudio) 7.0 0.0 "Linear" :!class /Script/Engine.AudioComponent)
            (bind _snow (Variables|Default|GetSnow)
              (:then
                (Components|Activation|SetActive _snow false false))
              (:else))))))
    (:False
      (Variables|Default|SetMoveProgress 1.01)
      (bind _asfort_game_mode_zone (Utilities|Casting|CastToFortGameModeZone (Game|GetGameMode)))
      (EventDispatchers|BindEventtoOnHandleMatchHasStarted _asfort_game_mode_zone (EventDispatchers|CreateEvent :!event Custom|ContinueSetup)))))

(event Custom|ContinueSetup
  (bind _6_31endpoint (Variables|Default|Get6-31EndPoint))
  ;[1 1 1] Calendar Event
  (Utilities|FlowControl|Delay 1.0)
  (Utilities|FlowControl|Sequence
    (:then_0
      (bind (_eventname _timeuntilend _timesincebegin) (MCPContext|WatchCalendarEvent (Variables|Default|GetFogApproachEvent))
        (:OnEventActive
          (Variables|Default|SetTotalEventTime (Math|Timespan|Timespan+Timespan _timeuntilend _timesincebegin))
          (Variables|Default|SetMoveProgress (Math|Float|MapRangeClamped (Math|Timespan|TimespanRatio _timeuntilend (Variables|Default|GetTotalEventTime)) 1.0 0.0 1.0 _6_31endpoint)) ; 0 = end point
        )))
    (:then_1
      (MCPContext|WatchCalendarEvent (Variables|Default|GetFogApproachOverlapEvent)
        (:OnEventActive
          (Variables|Default|SetMoveProgress _6_31endpoint))))))

(fn OnRep_MoveProgress
  (bind _ambaudio (Variables|Default|GetAmbAudio))
  (Utilities|FlowControl|Sequence
    (:then_0
      (if (< (Variables|Default|GetMoveProgress) 0.2)
        (Variables|Default|GetSnow
          (:then)
          (:else
            (bind _returnvalue (Effects|SpawnEmitteratLocation "/Figment_S06/Effects/Fort_Effects/World/Enviroment/Snow/P_Snow_Ambient_Island.P_Snow_Ambient_Island" "-131740.000000,-113900.000000,4920.000000"))
            (Variables|Default|SetSnow _returnvalue)))))
    (:then_1
      (CallFunction|UpdateIcebergLocation (Variables|Default|GetMoveProgress))
      (Audio|Parameter|SetFloatParameter _ambaudio "Progress" (- 1.0 (Variables|Default|GetMoveProgress)) :!class /Script/Engine.AudioComponent)
      (if (not (Audio|Components|Audio|IsPlaying _ambaudio))
        (Audio|Components|Audio|Play _ambaudio 0.0)))))

(fn UpdateIcebergLocation
  (params (Time :type Float%20%28double-precision%29))
  (bind _returnvalue (Math|Float|MapRangeClamped Time 0.1 0.02 0.0 1.0))
  (bind _returnvalue_2 (Math|Float|MapRangeClamped Time 0.2 0.0 0.0 1.0))
  (bind _icechunks (Variables|Default|GetIceChunks))
  (bind _aurora (Variables|Default|GetAurora))
  (Transformation|SetWorldLocation (Variables|Default|GetStaticMesh) (Spline|GetLocationatTime (Variables|Default|GetSpline) Time "World" true) false false)
  (Rendering|Material|SetScalarParameterValueonMaterials (Variables|Default|GetSM_Ice_Fog_Top) "Distance" (Math|Float|MapRangeClamped Time 1.0 0.0 0.4))
  (Transformation|SetRelativeLocation (Variables|Default|GetSM_Iceberg) (Math|Vector|Lerp(Vector) "-180.000000,-20.000000,-51.000000" "-370.000000,-30.000000,-80.000000" Time) false false)
  (Rendering|Material|SetScalarParameterValueonMaterials (Variables|Default|GetSM_Distant_Rock_Surround) "Reveal" _returnvalue_2)
  (Transformation|SetRelativeScale3D (Variables|Default|GetIceFloeParent) (Math|Conversions|ToVector(Float) (Math|Float|Lerp 1.0 0.2 Time)))
  (Rendering|SetVisibility _icechunks (< Time 0.12) true)
  (Transformation|SetWorldLocation _icechunks (Math|Vector|Lerp(Vector) "-480000.000000,-320000.000000,-5150.000000" "-200000.000000,-180000.000000,-5150.000000" _returnvalue_2) false false)
  (Rendering|SetVisibility _aurora (> _returnvalue 0.0) false)
  (Rendering|Material|SetScalarParameterValueonMaterials _aurora "Opacity" (Math|Float|Clamp(Float) _returnvalue 0.2)))

(fn ConstructionScript
  (CallFunction|UpdateIcebergLocation (Variables|Default|GetPreviewMoveProgress)))
