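; Celebration fireworks actor that plays burst effects and explosion sounds, driven by SpawnFireworks and StopFireWorks events.
; Start and stop control one of two timelines chosen by platform (mobile or other), with the non-mobile one set to quarter speed.
; Small bursts get random scale and offset around the actor, X and Y sharing one random value.
; The FireLama track spawns a huge burst high above, rotated toward the first local player with clamped pitch.
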(event Collision|EventActorBeginOverlap
  (params (OtherActor :type Actor%20Object%20Reference)))

(event EventTick
  (params (DeltaSeconds :type Float%20%28single-precision%29)))

(event EventBeginPlay)

(event Custom|SpawnFireworks
  (bind _returnvalue (Fortnite|GetLocalFortPlayerControllers true))
  (bind _returnvalue_2 (Utilities|Array|IsValidIndex _returnvalue))
  (bind _returnvalue_3 (Transformation|GetActorLocation))
  (bind _returnvalue_4 (Math|Vector|MakeVector 0.0 0.0 5e+03))
  (bind _returnvalue_5 (+ _returnvalue_3 _returnvalue_4))
  (bind (_returnvalue_roll _returnvalue_pitch _returnvalue_yaw) (Math|Rotator|BreakRotator (Math|Rotator|FindLookatRotation _returnvalue_5 (Variables|Default|GetPlayerlocation))))
  (bind _offsetrangez (Variables|Default|GetOffsetRangeZ))
  (bind _returnvalue_6 (* _offsetrangez -1.0))
  (bind _returnvalue_7 (Math|Random|RandomFloatinRange _returnvalue_6 _offsetrangez))
  (bind _returnvalue_8 (Transformation|GetActorLocation))
  (bind _offsetrangexy (Variables|Default|GetOffsetRangeXY))
  (bind _returnvalue_9 (* _offsetrangexy -1.0))
  (bind _returnvalue_10 (Math|Random|RandomFloatinRange _returnvalue_9 _offsetrangexy))
  (bind _returnvalue_11 (Math|Vector|MakeVector _returnvalue_10 _returnvalue_10 _returnvalue_7))
  (bind _returnvalue_12 (+ _returnvalue_8 _returnvalue_11))
  (if (Fortnite|IsMobilePlatform)
    (Timeline|Timeline_1 Play
      (:Fire
        _)
      (:FireLama
        (if _returnvalue_2
          (Variables|Default|SetPlayerlocation (Transformation|GetActorLocation (Utilities|Array|Get(acopy) _returnvalue))))
        (Effects|SpawnEmitteratLocation "/Figment_S06/Athena/Effects/Particles/Firework/P_HugeFirework_Explosion_Athena.P_HugeFirework_Explosion_Athena" _returnvalue_5 (Math|Rotator|SelectRotator (Math|Rotator|MakeRotator 0.0 (Math|Float|Clamp(Float) _returnvalue_pitch -7e+01 7e+01) _returnvalue_yaw) "0, 0, 0" _returnvalue_2))
        (Audio|PlaySoundatLocation "/Figment_S06/Sounds/Fort_GamePlay_Sounds/Explosives/Firework_Launcher_Explosion_CelebrCue.Firework_Launcher_Explosion_CelebrCue" _returnvalue_5)))
    (else
      (Timeline|Timeline_0 Play
        (:Fire
          _)
        (:FireLama
          _))))
  (Effects|SpawnEmitteratLocation :EmitterTemplate "/Figment_S06/Athena/Effects/Particles/Firework/P_BigFirework_Explosion_Athena.P_BigFirework_Explosion_Athena" :Location _returnvalue_12 :Scale (Math|Vector|MakeVector (Math|Random|RandomFloatinRange 1.0 0.2) (Math|Random|RandomFloatinRange 1.0 0.2) (Math|Random|RandomFloatinRange 1.0 0.2)) :bAutoDestroy true :PoolingMethod "None" :bAutoActivateSystem true)
  (Audio|PlaySoundatLocation "/Figment_S06/Sounds/Fort_GamePlay_Sounds/Explosives/Firework_Launcher_Explosion_CelebrCue.Firework_Launcher_Explosion_CelebrCue" _returnvalue_12))

(event Custom|StopFireWorks
  (bind _returnvalue (Fortnite|GetLocalFortPlayerControllers true))
  (bind _returnvalue_2 (Utilities|Array|IsValidIndex _returnvalue))
  (bind _returnvalue_3 (Transformation|GetActorLocation))
  (bind _returnvalue_4 (Math|Vector|MakeVector 0.0 0.0 5e+03))
  (bind _returnvalue_5 (+ _returnvalue_3 _returnvalue_4))
  (bind (_returnvalue_roll _returnvalue_pitch _returnvalue_yaw) (Math|Rotator|BreakRotator (Math|Rotator|FindLookatRotation _returnvalue_5 (Variables|Default|GetPlayerlocation))))
  (bind _offsetrangez (Variables|Default|GetOffsetRangeZ))
  (bind _returnvalue_6 (* _offsetrangez -1.0))
  (bind _returnvalue_7 (Math|Random|RandomFloatinRange _returnvalue_6 _offsetrangez))
  (bind _returnvalue_8 (Transformation|GetActorLocation))
  (bind _offsetrangexy (Variables|Default|GetOffsetRangeXY))
  (bind _returnvalue_9 (* _offsetrangexy -1.0))
  (bind _returnvalue_10 (Math|Random|RandomFloatinRange _returnvalue_9 _offsetrangexy))
  (bind _returnvalue_11 (Math|Vector|MakeVector _returnvalue_10 _returnvalue_10 _returnvalue_7))
  (bind _returnvalue_12 (+ _returnvalue_8 _returnvalue_11))
  (if (Fortnite|IsMobilePlatform)
    (Timeline|Timeline_1 Stop)
    (else
      (Timeline|Timeline_0 Stop)))
  (Effects|SpawnEmitteratLocation :EmitterTemplate "/Figment_S06/Athena/Effects/Particles/Firework/P_BigFirework_Explosion_Athena.P_BigFirework_Explosion_Athena" :Location _returnvalue_12 :Scale (Math|Vector|MakeVector (Math|Random|RandomFloatinRange 1.0 0.2) (Math|Random|RandomFloatinRange 1.0 0.2) (Math|Random|RandomFloatinRange 1.0 0.2)) :bAutoDestroy true :PoolingMethod "None" :bAutoActivateSystem true)
  (Audio|PlaySoundatLocation "/Figment_S06/Sounds/Fort_GamePlay_Sounds/Explosives/Firework_Launcher_Explosion_CelebrCue.Firework_Launcher_Explosion_CelebrCue" _returnvalue_12))

(fn ConstructionScript
  (Components|Timeline|SetPlayRate (Variables|BP_Fireworks_Celebration|GetTimeline_0) 0.25))
