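; BP_Fireworks_EVENT runs the timed celebration fireworks event in a match, driving the BP_Fireworks_Celebration actors in the level.
; On BeginPlay it waits for the start time from the AthenaGameDataUniversal curve table if the celebration is enabled there; a Debug flag substitutes a fixed delay.
; It then gathers every celebration actor, calls SpawnFireworks on each, and calls StopFireWorks after the table-driven duration.
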(event EventBeginPlay
  (if (Variables|Default|GetDebug)
    (Utilities|FlowControl|Delay 1e+01)
    (else
      (bind _outxy (DataTable|EvaluateCurveTableRow "/Figment_S06/Athena/Balance/DataTables/AthenaGameDataUniversal.AthenaGameDataUniversal" "Default.Celebration.Enabled")
        (:RowFound
          (if (== _outxy 1.0)
            (bind _outxy_2 (DataTable|EvaluateCurveTableRow "/Figment_S06/Athena/Balance/DataTables/AthenaGameDataUniversal.AthenaGameDataUniversal" "Default.Celebration.TimeBeforeStart")
              (:RowFound
                (Utilities|FlowControl|Delay _outxy_2))) ; TimeIntotheMatch
          ))) ; IsEventAcitve
    ))
  (bind _outactors (Actor|GetAllActorsOfClass "/Figment_S06/Athena/Environments/Blueprints/Events/BP_Fireworks_Celebration.BP_Fireworks_Celebration_C"))
  (for _array_element _outactors
    (Class|BPFireworksCelebration|SpawnFireworks _array_element))
  (bind _outxy_3 (DataTable|EvaluateCurveTableRow "/Figment_S06/Athena/Balance/DataTables/AthenaGameDataUniversal.AthenaGameDataUniversal" "Default.Celebration.DurationAfterStart")
    (:RowFound
      (Utilities|FlowControl|Delay _outxy_3)
      (for _array_element_2 _outactors
        (Class|BPFireworksCelebration|StopFireWorks _array_element_2)))) ; DurationoftheFireWorks
)

(event Collision|EventActorBeginOverlap
  (params (OtherActor :type Actor%20Object%20Reference)))

(event EventTick
  (params (DeltaSeconds :type Float%20%28single-precision%29)))

(fn ConstructionScript)
